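# Translucent UI material for icons, filed under a ToBeDeprecated folder, with a swappable Texture parameter that defaults to a global chevron.
# The 9-Slice UV output of GetUserInterfaceUV passes through CustomRotator, driven by a 0-1 Rotation parameter, before sampling the texture as alpha.
# The graph shows no connection from the sample to a material output.
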Material /FortUILibrary/Textures/Icons/ToBeDeprecated/M_UI_Icon__Mat
graph {
 properties {
  bUsedWithStaticMesh=false
  blendMode=BLEND_Translucent
  materialDomain=MD_UI
  pixelDepthOffsetMode=PDOM_Legacy
  usedShadingModels=Default Lit
 }
 nodes [TextureSampleParameter2D(UVs=MaterialFunctionCall(
    UVs=MaterialFunctionCall() {materialFunction=&GetUserInterfaceUV}.9-Slice UV
    "Rotation Angle (0-1)"=ScalarParameter() {parameterName=Rotation, sliderMax=1}
   ) {materialFunction=&CustomRotator}.Rotated Values) {parameterName=Texture, samplerType=SAMPLERTYPE_Alpha, texture=&/T_UI_Icon_Global_Chevron}]
}
